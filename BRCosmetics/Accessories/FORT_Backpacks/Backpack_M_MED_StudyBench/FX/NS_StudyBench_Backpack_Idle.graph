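# Niagara idle effect for the StudyBench backpack that sheds mesh petals from a skeletal mesh.
# Spawn rate rises with owner speed and has lower probability while skydiving; User.LandBurst triggers extra petals, scaled by a Timeline so repeated bursts shrink.
# Petals tumble and fade under masked curl noise, gravity and drag, following source movement early in life.
# Module-namespace copies of particle attributes supply a black, enlarged variant alongside two mesh renderers.
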NiagaraSystem /BRCosmetics/Accessories/FORT_Backpacks/Backpack_M_MED_StudyBench/FX/NS_StudyBench_Backpack_Idle
dependencies {
 usedDataInterfaces [&NiagaraDataInterfaceCurve, &NiagaraDataInterfaceVectorCurve]
 usedDynamicInputs []
 usedModules [
  &self:SystemSpawnScript.SystemScriptSource.SystemScriptGraph.NiagaraNodeAssignment_3.SetVariables_#1
  &SystemState
  &self:SystemSpawnScript.SystemScriptSource.SystemScriptGraph.NiagaraNodeAssignment_8.SetVariables_#2
  &self:SystemSpawnScript.SystemScriptSource.SystemScriptGraph.NiagaraNodeAssignment_9.SetVariables_#3
 ]
}
emitters {Petals {
  emitterSpawnScript [/SetVariables(Emitter.Delay=1) {isSetParameters=true}]
  emitterUpdateScript [
   &EmitterState()
   &SpawnRate(
    Spawn Probability=&MakeCustomFloatFromBool(Bool=$User.IsSkydiving, True Float=0.5, False Float=1)
    SpawnRate=&Multiply_Float(
     A=&Lerp_Float(
      A=10
      B=15
      Alpha=&ClampFloat(Float=&Divide_Float(A=&VectorLength(Vector=$Engine.Owner.Velocity), B=500))
     )
     B=$System.SpawnRateCurve
    )
    Use Spawn Probability=true
   ) {module=SpawnRate001}
   &SpawnPerFrame(
    Spawn Count:NiagaraInt32=&MakeIntFromFloat(Float=&Multiply_Float(
      A=&RandomRangeFloat(Minimum=7, Maximum=12)
      B=&FloatFromCurve(
       FloatCurve=DataInterface() {propertyValues {Curve {keys [
           {arriveTangent=0.5131, interpMode=RCIM_Cubic, leaveTangent=0.5131, tangentMode=RCTM_User, time=0, value=0}
           {arriveTangent=1.2242, interpMode=RCIM_Cubic, leaveTangent=1.2242, tangentMode=RCTM_User, time=0.5, value=0.3}
           {arriveTangent=1.9063, interpMode=RCIM_Cubic, leaveTangent=1.9063, tangentMode=RCTM_User, time=1, value=1}
          ]}}}
       CurveIndex=$Emitter.Delay
      )
     ))
    Spawn Probability:NiagaraFloat=$User.LandBurst
    Use Spawn Probability=true
   )
   &Timeline(
    Play=&SetBoolByFloatComparison(A=$User.LandBurst, B=1, Comparison Type=::NewEnumerator4)
    Play Rate=20
    Rewind Rate=0.5
   )
   /SetVariables(Emitter.Delay:NiagaraFloat=$Emitter.Timeline.InverseNormalizedTime) {isSetParameters=true}
  ]
  particleSpawnScript [
   &InitializeParticle(
    Lifetime Min=1.2
    Lifetime Mode=::NewEnumerator1
    Mass Mode=::NewEnumerator3
    Mesh Scale Mode=::NewEnumerator3
    Mesh Uniform Scale=0.06
    Sprite Rotation Mode=::NewEnumerator2
    UsePositionOffset=false
    Write Material Random=true
   )
   &InitialMeshOrientation(Rotation=&RandomRangeVector(Minimum=<0 0 0>, Maximum=<1 1 1>), Use Rotation Vector=true)
   &AddVelocity(Velocity=&Lerp_Vector(
     A=&RandomRangeVector(Minimum=<50 50 25>, Maximum=<-50 -50 25>)
     B=&RandomRangeVector(Minimum=<-15 -15 -25>, Maximum=<50 50 50>)
     Alpha=$User.LandBurst
    ))
   &ApplyInitialForces(Apply Force to Position=true)
   &SkeletalMeshLocation(
    Bone Sampling Mode=::NewEnumerator4
    MaskSpawnGroup=false
    Sampled Position Offset=<5 -2 3>
    Sampled Position Offset Coordinate Space:ENiagaraSkelMeshTransforms=::NewEnumerator2
    Skeletal Mesh=$System.Skeletal Mesh
   )
   &ShapeLocation(MaskSpawnGroup=false, Override Seed=false, Sphere Radius=6, UseOwnerScale=true)
   &CameraOffset(Camera Offset Amount=5)
   /SetVariables(Particles.Rot:NiagaraFloat=&UniformAOrBFloat(A=-1, Evaluation Type=::NewEnumerator0, Recalculate Random Each Loop=true, Override Seed=false)) {isSetParameters=true}
  ]
  particleUpdateScript [
   &ParticleState()
   &InheritSourceMovement(
    Applied Position Scale=&VectorFromCurve(VectorCurve=DataInterface() {propertyValues {XCurve {keys [<at 0.5 1>,<at 1 0>]}, YCurve {keys [<at 0.5 1>,<at 1 0>]}, ZCurve {keys [<at 0.5 1>,<at 1 0>]}}})
    ApplyForce=false
    ApplyPosition=true
    ApplyPositionFalloff=false
   )
   &CurlNoiseForce(
    Curl Noise Cone Mask Angle=60
    Curl Noise Cone Mask Falloff Angle=60
    Mask Curl Noise=true
    Noise Frequency=15
    Noise Strength=&RandomRangeFloat(Minimum=100, Maximum=200)
    Pan Noise=false
   )
   &GravityForce(Gravity=<0 0 -150>)
   &Drag(Drag=&RandomRangeFloat(Minimum=2.25, Maximum=2.75), Use Linear Drag=true, Use Rotational Drag=false)
   &ScaleColor(
    Scale Alpha=&FloatFromCurve(FloatCurve=DataInterface() {propertyValues {Curve {keys [{interpMode=RCIM_Cubic, tangentMode=RCTM_User, time=0, value=1}, {interpMode=RCIM_Cubic, tangentMode=RCTM_User, time=1, value=0}]}}})
    ScaleA=true
    ScaleRGB=true
   )
   &ScaleSpriteSize(Uniform Curve Sprite Scale=DataInterface() {propertyValues {Curve {keys [
        {arriveTangent=-0.0276, interpMode=RCIM_Cubic, leaveTangent=-0.0276, tangentMode=RCTM_User, time=0, value=0.5}
        {arriveTangent=0.0245, interpMode=RCIM_Cubic, leaveTangent=0.0245, tangentMode=RCTM_User, time=0.1408, value=0.9847}
        {arriveTangent=-1.146, interpMode=RCIM_Cubic, leaveTangent=-1.146, time=0.8, value=0.75}
        {arriveTangent=-0.0405, interpMode=RCIM_Cubic, leaveTangent=-0.0405, tangentMode=RCTM_User, time=1, value=0}
       ]}}})
   &SolveForcesAndVelocity(Clamp Velocity=false, Limit Acceleration=false)
   &FN_Sprite_Attributes(
    Color=&MultiplyLinearColor(Linear Color=$Particles.Color, Scale Factor=<0 0 0 1>)
    Create Camera Offset Attribute=true
    Create Color Attribute=true
    Create SpriteSize Attribute=true
    SpriteSize=&Multiply_Vector2D(A=$Particles.SpriteSize, B=<1.3 1.3>)
    Write to Module Namespace=true
   )
   &UpdateMeshOrientation(Rotation Rate=$Particles.Rot, Rotation Vector=&RandomRangeVector(Minimum=<0 0 0>, Maximum=<0 0 1>))
   &ScaleMeshSize(Scale Factor=&VectorFromFloat(Value=&FloatFromCurve(FloatCurve=DataInterface() {propertyValues {Curve {keys [
          {arriveTangent=-0.0276, interpMode=RCIM_Cubic, leaveTangent=-0.0276, tangentMode=RCTM_User, time=0, value=0}
          {arriveTangent=0.0245, interpMode=RCIM_Cubic, leaveTangent=0.0245, tangentMode=RCTM_User, time=0.1408, value=0.9847}
          {arriveTangent=-1.146, interpMode=RCIM_Cubic, leaveTangent=-1.146, time=0.8, value=0.75}
          {arriveTangent=-0.0405, interpMode=RCIM_Cubic, leaveTangent=-0.0405, tangentMode=RCTM_User, time=1, value=0}
         ]}}})))
   &FN_Mesh_Attributes(
    Create Scale Attribute=true
    Scale=&Multiply_Vector(A=$Particles.Scale, B=&Multiply_VectorByFloat(Vector=<1 1 1>, Float=1.2))
    Write to Module Namespace=true
   )
  ]
  rendererClasses [&NiagaraMeshRendererProperties]
  renderers [
   {rendererClass=&NiagaraMeshRendererProperties, rendererIndex=0}
   {rendererClass=&NiagaraMeshRendererProperties, rendererIndex=1}
  ]
 }}
summary {
 systemName=NS_StudyBench_Backpack_Idle
 emitters [{emitterName=Petals, rendererClasses [&NiagaraMeshRendererProperties]}]
 userVariables [
  User.bIsFrontEnd {type=NiagaraBool}
  User.bIsFrontEndPreview {type=NiagaraBool}
  User.IsSkydiving {type=NiagaraBool}
  User.LandBurst {type=NiagaraFloat, defaultValue=0, description=""}
  User.SkeletalMesh_01 {
   type=NiagaraDataInterfaceSkeletalMesh
   defaultValue {
    struct=NiagaraExt_VariableValue_DataInterface
    value {dataInterface=&self:NiagaraDataInterfaceSkeletalMesh_0}
   }
  }
 ]
}
